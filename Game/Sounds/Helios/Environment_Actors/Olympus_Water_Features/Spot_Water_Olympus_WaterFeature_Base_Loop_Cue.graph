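# Sound cue for the looping base ambience of Olympus water features, playing a small, close water spot recording at low volume.
# A Modulator applies slight random pitch variation, and every quality level routes to that same branch.
# It borrows attenuation from a Brimstone waterfall impact, overrides concurrency to cap simultaneous voices, and uses the Fort_Gadgets class.
# Sends feed the exterior ambient audio bus and default reverb submix.
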SoundCue /Game/Sounds/Helios/Environment_Actors/Olympus_Water_Features/Spot_Water_Olympus_WaterFeature_Base_Loop_Cue
root=QualityLevel_0
nodes [
 {id=Modulator_1, children [WavePlayer_6], properties {pitchMax=1.1, pitchMin=0.9, volumeMax=1, volumeMin=1}}
 {id=QualityLevel_0, children [Modulator_1,Modulator_1,Modulator_1,Modulator_1]}
 {
  id=WavePlayer_6
  properties {bLooping=true, soundWaveAssetPtr=&/Spot_Olympus_Water_Feature_Small_Close_01}
 }
]
properties {
 attenuationSettings=/Game/Sounds/Helios/Environment_Actors/BrimstoneWaterfall/Spot_Water_Waterfall_Goopy_Impact_Heavy_Occ_Att
 bOverrideConcurrency=true
 maxDistance=2000
 priority=10
 soundClassObject=&Fort_Gadgets
 volumeMultiplier=0.16
 busSends [{audioBus=null, minSendLevel=1, sendLevel=1, soundSourceBus=&AmbientAudio_Exterior_SourceBus}]
 concurrencyOverrides {maxCount=2, voiceStealReleaseTime=2}
 soundSubmixSends [{minSendLevel=1, soundSubmix=&Fortnite_Reverb_Default}]
}
