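# Looping ambient spot sound cue for a mid-sized Olympus water feature, built on the Small_01 water wave.
# A modulator varies pitch around normal, feeding a random short start delay, then an enveloper that fades volume in over three seconds on the looping wave.
# It borrows Brimstone Waterfall attenuation, uses the indoor ambient sound class, sends to the default reverb submix and caps simultaneous voices by priority.
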SoundCue /Game/Sounds/Helios/Environment_Actors/Olympus_Water_Features/Spot_Water_Olympus_WaterFeature_Mid_Loop_Cue
root=Modulator_1
nodes [
 {id=Delay_1, children [Enveloper_1], properties {delayMax=1.5}}
 {
  id=Enveloper_1
  children [WavePlayer_2]
  properties {
   pitchCurve {editorCurveData {keys [{time=0}]}}
   volumeCurve {editorCurveData {keys [<at -0.0001 0 RCIM_Cubic>, {arriveTangent=0.4839, interpMode=RCIM_Cubic, leaveTangent=0.4839, tangentMode=RCTM_User, time=3, value=1}]}}
  }
 }
 {id=Modulator_1, children [Delay_1], properties {pitchMax=1.1, pitchMin=0.9, volumeMax=1, volumeMin=1}}
 {
  id=WavePlayer_2
  properties {bLooping=true, soundWaveAssetPtr=&/Spot_Olympus_Water_Feature_Small_01}
 }
]
properties {
 attenuationSettings=/Game/Sounds/Helios/Environment_Actors/BrimstoneWaterfall/Spot_BrimstoneWaterfall_MainFlow_Small_Occ_Att
 bOverrideConcurrency=true
 maxDistance=750
 priority=10
 soundClassObject=&Fort_Ambient_Indoors_Athena
 volumeMultiplier=0.626
 concurrencyOverrides {maxCount=3, resolutionRule=StopLowestPriority, voiceStealReleaseTime=1}
 soundSubmixSends [{minSendLevel=1, soundSubmix=&Fortnite_Reverb_Default}]
}
